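; Molotov Cocktail (Firefly Jar) grenade projectile that bursts into smaller fireball projectiles on impact.
; Entering water before a hit kills the flame, hides the projectile and respawns it as a pickup.
; With Curie active, a hit makes the server explode, defer-spawn a fire BGA and launch fireballs, while clients play cosmetics.
; A scalable float selects timer-spawned homing fireballs fanned along the reflection vector, or cone-randomized "drunk" movement ones.
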(event EventBeginPlay
  ;[1 1 1] On Begin Play, turn on Fire Particle FX for Burning Rag - Bind Event to Enter Water - On Water Enter, disable Fire Particle FX and then hide the projectile and spawn a new pickup
  (Curie|OnCurieActivated
    (:then
      (Components|Activation|Activate (Variables|Default|GetNiagara_Projectile_Particles) false)
      ;;[1 1 1] This is to fix throwing the projectile whilst in water, to allow it time to leave the pawns hands
      (Utilities|FlowControl|Delay 0.2)
      (Variables|Default|SetCanInteractWithWater true)
      (if (Water|IsinWater (Variables|Water|GetWaterInteractionComponent))
        (CallFunction|HasEnteredWater)))
    (:OnCurieActive
      (Variables|Default|SetCurieActivated true))))

(event Custom|HasEnteredWater
  ;[1 1 1] On Begin Play, turn on Fire Particle FX for Burning Rag - Bind Event to Enter Water - On Water Enter, disable Fire Particle FX and then hide the projectile and spawn a new pickup
  ;;[1 1 1] If already entered Hit, don't do anything with the water interaction. Set Entered Water so it can be checked in Hit interaction.
  (if (not (Variables|Default|GetHit))
    (Utilities|FlowControl|DoOnce false
      (:Completed
        (Variables|Default|SetEnteredWater true)
        (Utilities|FlowControl|Sequence
          (:then_0
            (Components|Activation|Deactivate (Variables|Default|GetNiagara_Projectile_Particles))
            (Rendering|SetActorHiddenInGame true)
            (Inventory|SpawnPickupInWorld "/Game/Athena/Items/Consumables/MolotovCocktail/WID_Athena_Grenade_Molotov.WID_Athena_Grenade_Molotov" 1 (Transformation|GetWorldLocation (Variables|Collision|GetCapsuleComponent)) "0.000000,0.000000,0.000000" 0 true false)
            (Actor|SetLifeSpan 0.5))
          (:then_1
            (bind _grenadefuse_audiocomponent (Variables|B_Prj_Grenade|GetGrenadeFuse_AudioComponent)
              (:then
                (Audio|Components|Audio|FadeOut _grenadefuse_audiocomponent 0.2 0.0 "Linear" :!class /Script/Engine.AudioComponent))
              (:else)))
          (:then_2
            (Effects|Components|ParticleSystem|SpawnFXAtLocation "/Game/Athena/Items/Consumables/MolotovCocktail/NS_FireflyJar_WaterSplash.NS_FireflyJar_WaterSplash" (Transformation|GetWorldLocation (Variables|Collision|GetCapsuleComponent)))))))))

(event OnEnterWaterEvent(WaterInteractionComponent)
  (params (WaterBody :type Fort%20Water%20Body%20Actor%20Object%20Reference) (WaterInteractionComponent :type Fort%20Water%20Interaction%20Component%20Object%20Reference) (bIsFirstBody :type Boolean))
  ;[1 1 1] On Begin Play, turn on Fire Particle FX for Burning Rag - Bind Event to Enter Water - On Water Enter, disable Fire Particle FX and then hide the projectile and spawn a new pickup
  (if (Variables|Default|GetCanInteractWithWater)
    (CallFunction|HasEnteredWater)))

(event EventHeldWaterImpact
  ;[1 1 1] On Begin Play, turn on Fire Particle FX for Burning Rag - Bind Event to Enter Water - On Water Enter, disable Fire Particle FX and then hide the projectile and spawn a new pickup
  (|CallParentFunction)
  (bind _grenadefuse_audiocomponent (Variables|B_Prj_Grenade|GetGrenadeFuse_AudioComponent)
    (:then
      (Audio|Components|Audio|FadeOut _grenadefuse_audiocomponent 0.25 0.0 "Linear" :!class /Script/Engine.AudioComponent))
    (:else)))

(event Collision|EventHit
  (params (MyComp :type Primitive%20Component%20Object%20Reference) (Other :type Actor%20Object%20Reference) (OtherComp :type Primitive%20Component%20Object%20Reference) (bSelfMoved :type Boolean) (HitLocation :type Vector :default "0, 0, 0") (HitNormal :type Vector :default "0, 0, 0") (NormalImpulse :type Vector :default "0, 0, 0") (Hit :type Hit%20Result%20Structure%20%28by%20ref%29))
  (bind _hit_location (Variables|Default|GetHit%20Location))
  ;[1 1 1] On hit, store the Hit Location / Hit Normal then mirror the projectile velocity and hit normal to get the reflection vector
  ;;[1 1 1] If already entered water, don't do anything with this hit. Set hit so it can be checked in water interaction.
  (if (not (Variables|Default|GetEnteredWater))
    (Utilities|FlowControl|DoOnce false
      (:Completed
        (Variables|Default|SetHit true)
        (Variables|Default|SetReflectionVector (Math|Vector|MirrorVectorbyNormal (Variables|/Script/Engine.MovementComponent|GetVelocity (Variables|Movement|GetProjectileMovementComponent)) HitNormal))
        (Variables|Default|SetHit%20Normal HitNormal)
        (bind _output_get (Variables|Default|SetHit%20Location HitLocation))
        (Utilities|FlowControl|DoOnce false
          (:Completed
            (Networking|SwitchHasAuthority
              (:Authority
                ;;[1 1 1] Check if Curie is active
                (if (Variables|Default|GetCurieActivated)
                  (Variables|Default|SetSpawnBGATransform (Math|Transform|MakeTransform _output_get))
                  (Utilities|Time|SetTimerbyEvent (EventDispatchers|CreateEvent :!event Custom|Defer%20BGA%20Spawn) 0.1)
                  ;[1 1 1] Take the Refection vector and the hit normal to figure out the spawn directrions for the additional projectiles
                  (if (== (ScalableFloat|GetValueatLevel (Variables|Default|GetDoWeUseHomingProjectile)) 1.0)
                    (Variables|Default|SetFireflyProjectileDirection (Math|Vector|RotateVectorAroundAxis (Math|Vector|Normalize (Math|Vector|ProjectVectorontoPlane (Variables|Default|GetReflectionVector) (Variables|Default|GetHit%20Normal))) (* (Variables|Default|GetFireflyProjectileAngleExtreme) -1.0) (Variables|Default|GetHit%20Normal)))
                    (Variables|Default|SetNumProjectilesToSpawn (Math|Float|Truncate (ScalableFloat|GetValueatLevel (Variables|Default|GetFireBallsToSPawn))))
                    ;[1 1 1] Spawn the additional projectiles and then do our explosion at the main impact
                    (Utilities|Time|SetTimerbyFunctionName 0 "FireflyProjectileLogic" 0.15 true)
                    (Development|SetHiddeninGame (Variables|B_Prj_Grenade|GetMesh) true false)
                    (Delivery|DoExplosion)
                    (else
                      (CallFunction|DrunkProjectiles)))
                  (else
                    (Development|SetHiddeninGame (Variables|B_Prj_Grenade|GetMesh) true false))))
              (:Remote
                (if (Variables|Default|GetCurieActivated)
                  ;[1 0.6116 0.1752] FX Explosion
                  (Networking|CanExecuteCosmeticEvents
                    (:True
                      (Utilities|FlowControl|Sequence
                        (:then_0
                          (Components|Activation|Deactivate (Variables|Default|GetNiagara_Projectile_Particles)))
                        (:then_1
                          (Effects|Components|ParticleSystem|SpawnFXAtLocation "/Game/Athena/Items/Consumables/MolotovCocktail/NS_Firefly_Jar_Explosion.NS_Firefly_Jar_Explosion" _hit_location (Math|Conversions|RotationFromXVector (Variables|Default|GetHit%20Normal)) "1.000000,1.000000,1.000000" true "AutoRelease"))
                        (:then_2
                          (if (not (Variables|Default|GetEnteredWater))
                            (Audio|PlaySoundatLocation (Variables|Default|GetIgnite%20Audio%20Cue) (Variables|Default|GetHit%20Location))))
                        (:then_3
                          (Fortnite|PlayLocalForceFeedbackatLocationMulti "/Game/Weapons/ForceFeedback/FF_Generic_Explosion_Small_Near.FF_Generic_Explosion_Small_Near" "/Game/Weapons/ForceFeedback/FF_Generic_Explosion_Small_Far.FF_Generic_Explosion_Small_Far" _hit_location))
                        (:then_4
                          (Camera|PlayWorldCameraShake "/Game/Effects/Fort_Effects/BluePrints/CameraShakes/B_Medium_Explosion_CameraShake.B_Medium_Explosion_CameraShake_C" _hit_location))
                        (:then_5
                          (Fortnite|ShowSoundIndicatorLocal 0 _hit_location (Variables|/Script/Engine.SoundBase|GetMaxDistance (Variables|Default|GetIgnite%20Audio%20Cue)) "Gunshot")))))
                  (else
                    (Development|SetHiddeninGame (Variables|B_Prj_Grenade|GetMesh) true false)
                    (Components|Activation|Deactivate (Variables|Default|GetNiagara_Projectile_Particles))))))))))))

(event Custom|Defer%20BGA%20Spawn
  (Building|SpawnBuildingGameplayActor "/Game/Athena/Items/Consumables/MolotovCocktail/BGA_CurieFireMesh_MolotovCocktail.BGA_CurieFireMesh_MolotovCocktail_C" (Variables|Default|GetSpawnBGATransform) (Actor|GetOwner)))

(fn DrunkProjectiles
  (for _index (range (Math|Float|Truncate (ScalableFloat|GetValueatLevel (Variables|Default|GetFireBallsToSPawn))))
    ;[1 1 1] Drunk Behavior
    (bind _returnvalue (Projectile|SpawnProjectile :ProjectileClass "/Game/Athena/Items/Consumables/MolotovCocktail/B_Prj_Athena_Grenade_MolotovCocktail_AdditionalProjectile.B_Prj_Athena_Grenade_MolotovCocktail_AdditionalProjectile_C" :RequestedBy (Actor|GetOwner) :SpawnLocation (Transformation|GetWorldLocation (Variables|Collision|GetCapsuleComponent)) :SpawnRotation "0, 0, 0" :SpawnDirection (Math|Rotator|MakeRotfromX (Math|Random|RandomUnitVectorinConeinDegrees (+ (Utilities|Array|Get(acopy) (Variables|Default|GetProjectileSpawnVectors) _index) (Variables|Default|GetHit%20Normal)) (ScalableFloat|GetValueatLevel (Variables|Default|GetConeHalfAngleInDegrees)))) :bUseDefaultPhysics false :InitialSpeed (Math|Random|RandomFloatinRange (ScalableFloat|GetValueatLevel (Variables|Default|GetSpawnedProjectileMinSpeed)) (ScalableFloat|GetValueatLevel (Variables|Default|GetSpawnedProjectileMaxSpeed))) :GravityScale 0.5 :ChargePercent 0.0))
    (Projectile|SetProjectileMovementDrunkData _returnvalue (Utilities|Struct|MakeProjectileMovementDrunkConfig "/Game/Athena/Items/Consumables/MolotovCocktail/MolotovDrunkSpeedScaleCurve.MolotovDrunkSpeedScaleCurve" (select (Math|Random|RandomBool) "/Game/Athena/Items/Consumables/MolotovCocktail/MolotovDrunkGravityScaleCurve_Two.MolotovDrunkGravityScaleCurve_Two" "/Game/Athena/Items/Consumables/MolotovCocktail/MolotovDrunkGravityScaleCurve.MolotovDrunkGravityScaleCurve") 0.1 (Math|Random|RandomFloatinRange 0.5 4.0) 1.0 0.1 3e+01 0.5)))
  (Development|SetHiddeninGame (Variables|B_Prj_Grenade|GetMesh) true false)
  (Delivery|DoExplosion))

(fn FireflyProjectileLogic
  (bind _numprojectilespawned (Variables|Default|GetNumProjectileSpawned))
  (bind _result (Math|Integer|IncrementInt _numprojectilespawned)
    (:%20%20
      (if (<= _result (Variables|Default|GetNumProjectilesToSpawn))
        (Variables|Default|SetNewTemp (* (Math|Vector|RotateVectorAroundAxis (Variables|Default|GetFireflyProjectileDirection) (- (* _numprojectilespawned (/ (Variables|Default|GetFireflyProjectileAngleExtreme) 2.0)) (Math|Random|RandomFloatinRange 0.0 (/ (Variables|Default|GetFireflyProjectileAngleExtreme) 2.0))) (Variables|Default|GetHit%20Normal)) (Math|Random|RandomFloatinRange (Variables|Default|GetFireflyProjectileRadius_Min) (Variables|Default|GetFireflyProjectileRadius_Max))))
        (if (== (Variables|Default|GetNumProjectileSpawned) 1)
          _
          (elif (== (Variables|Default|GetNumProjectileSpawned) (Variables|Default|GetNumProjectilesToSpawn))
            _
            (else
              (Variables|Default|SetFireflyProjectileEndPoint (+ (Variables|Default|GetHit%20Location) (Variables|Default|GetNewTemp)))
              (bind _returnvalue (Projectile|SpawnProjectile :ProjectileClass "/Game/Athena/Items/Consumables/MolotovCocktail/B_Prj_Athena_Grenade_MolotovCocktail_AdditionalProjectile.B_Prj_Athena_Grenade_MolotovCocktail_AdditionalProjectile_C" :RequestedBy (Actor|GetOwner) :SpawnLocation (Transformation|GetWorldLocation (Variables|Collision|GetCapsuleComponent)) :SpawnRotation "0, 0, 0" :SpawnDirection (Math|Rotator|MakeRotfromX (Variables|Default|GetHit%20Normal)) :bUseDefaultPhysics false :InitialSpeed (Math|Random|RandomFloatinRange 3.5e+02 4.5e+02) :GravityScale 0.25 :ChargePercent 0.0))
              (Projectile|SetHomingData _returnvalue _returnvalue (Utilities|Struct|MakeProjectileHomingData "LockOn" 5e+01 1.5e+02 (Math|Random|RandomFloatinRange 0.75 1.0) 2e+01 -1 0 0 (Variables|Default|GetFireflyProjectileEndPoint)) :!class /Script/FortniteGame.FortProjectileMovementInterface))))
        (Variables|Default|SetNewTemp (* (Variables|Default|GetNewTemp) 0.5))
        (else
          (Utilities|Time|ClearTimerbyFunctionName 0 "FireflyProjectileLogic"))))))

(fn ConstructionScript)
