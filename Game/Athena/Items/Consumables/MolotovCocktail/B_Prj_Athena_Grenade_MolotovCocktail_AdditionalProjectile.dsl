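; Additional Molotov cocktail (Firefly Jar) grenade projectile that wanders through the air and spawns a Curie building gameplay actor where it lands.
; A looping randomized timer re-homes it toward random points below, while another timer fades and destroys it after several seconds.
; Hits stop movement, spawn the actor, play client-only explosion effects and destroy the projectile.
; Water entry follows a similar path but reuses EventHit's possibly unset outputs.
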(event Custom|HasNotLandedFadeOut
  ;[1 1 1] If we haven't landed, fade out and die.
  (Components|Activation|Deactivate (Variables|Default|GetFire_Trail_Niagara))
  (Audio|Components|Audio|FadeOut (Variables|Default|GetFireflyProjectileAudio) 0.4 0.0 "Linear" :!class /Script/Engine.AudioComponent)
  (Utilities|FlowControl|Delay 0.4)
  (Actor|DestroyActor))

(event EventBeginPlay
  ;[1 1 1] On spawn, wait a few seconds before we start a looping timer, and set a 2nd timer for fading out the projectile if it doesn't land
  (|CallParentFunction)
  (Development|PrintString "Spawned")
  (Utilities|Time|SetTimerbyEvent (delegate Custom|BeginDrunkenMovement) (Math|Random|RandomFloatinRange 0.3 0.6) true false 2.5 1.5)
  (Utilities|Time|SetTimerbyEvent (delegate Custom|HasNotLandedFadeOut) (Math|Random|RandomFloatinRange 4.0 6.0))
  ;[1 1 1] Turn off Capsule Collision for Perf
  (Networking|CanExecuteCosmeticEvents
    (:True
      (Variables|/Script/Engine.PrimitiveComponent|SetbGenerateOverlapEvents false (Variables|Collision|GetCapsuleComponent))
      (Collision|SetCollisionEnabled (Variables|Collision|GetCapsuleComponent) "NoCollision"))))

(event Custom|BeginDrunkenMovement
  ;[1 1 1] On a looping timer, create a new ground target location whilst in the air and home towards it.
  (Projectile|SetHomingData self self (Utilities|Struct|MakeProjectileHomingData "LockOn" (Math|Float|SelectFloat 5.5e+02 1.25e+03 (Variables|Default|GetIsHomingToGround)) (Math|Float|SelectFloat 5.5e+02 1.25e+03 (Variables|Default|GetIsHomingToGround)) 0.0 2e+01 -1 0 0 (- (+ (* (Math|Random|RandomUnitVector) 3.5e+03) (Transformation|GetActorLocation)) "0, 0,10000.000000")) :!class /Script/FortniteGame.FortProjectileMovementInterface)
  (Variables|Default|SetIsHomingToGround true))

(event Custom|Landed%20In%20Water
  ;[1 1 1] Spawn the curie BGA
  (Building|SpawnBuildingGameplayActor (Variables|Default|GetBGAToSpawn) (stale Collision|EventHit _returnvalue) (Actor|GetOwner))
  ;[1 1 1] Play the FX on clients and then destroy the actor. Give the FX time to play
  (CallFunction|PlayClientFX (stale Collision|EventHit HitLocation))
  (Utilities|FlowControl|Delay 0.5)
  (Actor|DestroyActor))

(event Collision|EventHit
  (params (MyComp :type Primitive%20Component%20Object%20Reference) (Other :type Actor%20Object%20Reference) (OtherComp :type Primitive%20Component%20Object%20Reference) (bSelfMoved :type Boolean) (HitLocation :type Vector :default "0, 0, 0") (HitNormal :type Vector :default "0, 0, 0") (NormalImpulse :type Vector :default "0, 0, 0") (Hit :type Hit%20Result%20Structure%20%28by%20ref%29))
  ;[1 1 1] On hit immediately stop simulation and set has hit to true
  (|CallParentFunction MyComp Other OtherComp bSelfMoved HitLocation HitNormal NormalImpulse Hit)
  (Game|Components|ProjectileMovement|StopSimulating (Variables|Movement|GetProjectileMovementComponent) Hit)
  (Variables|Default|SetHasHit true)
  ;[1 1 1] This attempts to adjust the angle of the mesh based on the floor angle.
  (bind _returnvalue (CallFunction|FindMeshAngle HitLocation HitNormal))
  ;[1 1 1] Spawn the curie BGA
  (Building|SpawnBuildingGameplayActor (Variables|Default|GetBGAToSpawn) _returnvalue (Actor|GetOwner))
  ;[1 1 1] Play the FX on clients and then destroy the actor. Give the FX time to play
  (CallFunction|PlayClientFX HitLocation)
  (Utilities|FlowControl|Delay 0.5)
  (Actor|DestroyActor))

(event Custom|PlayClientFX
  (params (HitLocationData :type Vector))
  ;[1 1 1] Client side FX
  (Networking|CanExecuteCosmeticEvents
    (:True
      ;;[1 1 1] Play Explode Sound
      (Audio|PlaySoundatLocation (Variables|Default|GetSound_Explode) HitLocationData)
      ;;[1 0.621 0.2134] FX Explosion
      (Effects|Components|ParticleSystem|SpawnFXAtLocation "/Game/Athena/Items/Consumables/MolotovCocktail/NS_Firefly_Jar_Explosion_Small.NS_Firefly_Jar_Explosion_Small" HitLocationData (Transformation|GetActorRotation))
      (Utilities|FlowControl|Delay 0.5)
      ;;[1 1 1] Turn off the particle
      (Components|Activation|Deactivate (Variables|Default|GetFire_Trail_Niagara))
      (Audio|Components|Audio|FadeOut (Variables|Default|GetFireflyProjectileAudio) 0.4 0.0 "Linear" :!class /Script/Engine.AudioComponent))))

(event OnEnterWaterEvent(WaterInteractionComponent)
  (params (WaterBody :type Fort%20Water%20Body%20Actor%20Object%20Reference) (WaterInteractionComponent :type Fort%20Water%20Interaction%20Component%20Object%20Reference) (bIsFirstBody :type Boolean))
  ;[1 1 1] Handle water interaction
  (CallFunction|LandedinWater))

(fn FindMeshAngle
  (params (HitLoc :type Vector) (HitNotmal :type Vector))
  (return (Math|Transform|SelectTransform (Math|Transform|MakeTransform (+ HitLoc (* HitNotmal (Variables|Default|GetFire%20Plane%20Spawn%20Offset)))) (Math|Transform|MakeTransform (Transformation|GetWorldLocation (Variables|Collision|GetCapsuleComponent))) (Variables|Default|GetDoesMeshAngleNeedAdjusted))))

(fn ConstructionScript)
