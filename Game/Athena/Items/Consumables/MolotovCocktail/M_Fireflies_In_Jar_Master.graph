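# Translucent unlit master material for glowing firefly sprites, used for the fireflies-in-jar consumable.
# World position offset combines the Sprite function's billboarding with a local-space sine/cosine XY orbit and Z wave, transformed to world.
# Vertex color dotted with parameter vectors varies time, frequency and height per firefly, so motion needs only the shader.
# A UV sphere mask shapes each dot, tinted by Color, with a powered copy scaled by Color alpha for opacity.
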Material /Game/Athena/Items/Consumables/MolotovCocktail/M_Fireflies_In_Jar_Master
graph {
 properties {blendMode=BLEND_Translucent, pixelDepthOffsetMode=PDOM_Legacy, shadingModel=MSM_Unlit, usedShadingModels=Unlit}
 nodes [
  n1 = VertexColor()
  n2 = DotProduct(A=@n1, B=VectorParameter() {group=WPO, parameterName=RandomTimeAndZ, defaultValue=<rgba -1.7 1.3 2.1 0>}.RGB)
  n3 = Add(A=Time(), B=@n2)
  n4 = DotProduct(A=@n1, B=VectorParameter() {group=WPO, parameterName=RandomFrequency, defaultValue=<rgba 0.75 0.97 0.47 0>}.RGB)
  n5 = Multiply(A=@n3, B=Add(A=ScalarParameter() {defaultValue=1, group=WPO, parameterName=XYFrequency}, B=@n4))
  n6 = Multiply(
   A=MaterialFunctionCall(R=Sine(None=@n5), G=Cosine(None=@n5), B=Constant()) {materialFunction=&AppendMany}.RGB
   B=ScalarParameter() {defaultValue=3.5, group=WPO, parameterName=Orbit_Scale}
  )
  n7 = Constant()
  n8 = ScalarParameter() {defaultValue=2, group=WPO, parameterName=ZWaveFrequency}
  n9 = Sine(None=Multiply(A=@n3, B=Add(A=@n8, B=@n4)))
  n10 = Multiply(
   A=MaterialFunctionCall(R=@n7, G=@n7, B=Add(A=@n9, B=@n2)) {materialFunction=&AppendMany}.RGB
   B=ScalarParameter() {defaultValue=3, group=WPO, parameterName=ZWaveScale}
  )
  Add(
   A=MaterialFunctionCall!(
    Center Location=WorldPosition().XYZ
    0-1 UVs
    XY scale=VectorParameter() {group=Color, parameterName=SpriteSize, defaultValue=<rgba 5 5 0 1>}.RGB
    OverridePivot
    "Normals (Optional)"
   ) {materialFunction=&Sprite}.World position offset
   B=Transform(None=Add(A=@n6, B=@n10)) {transformSourceType=TRANSFORMSOURCE_Local}
  )
  n11 = VectorParameter() {parameterName=Color, defaultValue=<rgba 2 0.3037 0.0729 1>}
  n12 = SphereMask(
   A=TextureCoordinate()
   B=Constant2Vector() {g=0.5, r=0.5}
   Radius=ScalarParameter() {defaultValue=0.5, group=Mask, parameterName=Radius}
   Hardness=ScalarParameter() {group=Mask, parameterName=Hardness}
  )
  Multiply(A=@n11.A, B=Power(Base=@n12, Exp=ScalarParameter() {defaultValue=2, group=Mask, parameterName=Sphere_Power_Exp}))
  Multiply(A=@n11.RGB, B=@n12)
 ]
}
